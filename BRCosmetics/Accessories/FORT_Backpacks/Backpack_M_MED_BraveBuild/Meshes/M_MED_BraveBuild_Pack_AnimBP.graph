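# Animation Blueprint for the BraveBuild backpack cosmetic that procedurally makes its three right and three left feather bones float.
# Each update, the event graph builds six X/Y offset vectors from sine waves of real time, with slightly different frequencies per feather and a time shift on the left side.
# The AnimGraph applies them to the input pose as additive bone-space translations through chained Modify Bone nodes.
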AnimBlueprint /BRCosmetics/Accessories/FORT_Backpacks/Backpack_M_MED_BraveBuild/Meshes/M_MED_BraveBuild_Pack_AnimBP
target_skeleton=&/M_MED_BraveBuild_Pack_Skeleton
graphs [
 AnimGraph {nodes [
   n1 = "Transform (Modify) Bone - Bone: dyn_R_feather_1"(
    ComponentPose=Local To Component(LocalPose=Input Pose())
    Alpha=1
    Translation="Get R FloaterOut A "() -> ("R FloaterOut A ":Vector)
    Rotation="0,0,0"
    Scale="1,1,1"
   ) {translationMode=BMM_Additive, translationSpace=BCS_BoneSpace, alphaBoolBlend {blendOption=Linear}, boneToModify {boneName=dyn_R_feather_1}}
   n2 = "Transform (Modify) Bone - Bone: dyn_R_feather_2"(ComponentPose=@n1, Alpha=1, Translation=Get R FloaterOut B() -> (R FloaterOut B:Vector), Rotation="0,0,0", Scale="1,1,1") {translationMode=BMM_Additive, translationSpace=BCS_BoneSpace, alphaBoolBlend {blendOption=Linear}, boneToModify {boneName=dyn_R_feather_2}}
   n3 = "Transform (Modify) Bone - Bone: dyn_R_feather_3"(ComponentPose=@n2, Alpha=1, Translation=Get R FloaterOutC() -> (R FloaterOutC:Vector), Rotation="0,0,0", Scale="1,1,1") {translationMode=BMM_Additive, translationSpace=BCS_BoneSpace, alphaBoolBlend {blendOption=Linear}, boneToModify {boneName=dyn_R_feather_3}}
   n4 = "Transform (Modify) Bone - Bone: dyn_L_feather_1"(ComponentPose=@n3, Alpha=1, Translation="Get L FloaterOut A "() -> ("L FloaterOut A ":Vector), Rotation="0,0,0", Scale="1,1,1") {translationMode=BMM_Additive, translationSpace=BCS_BoneSpace, alphaBoolBlend {blendOption=Linear}, boneToModify {boneName=dyn_L_feather_1}}
   n5 = "Transform (Modify) Bone - Bone: dyn_L_feather_2"(ComponentPose=@n4, Alpha=1, Translation=Get L FloaterOut B() -> (L FloaterOut B:Vector), Rotation="0,0,0", Scale="1,1,1") {translationMode=BMM_Additive, translationSpace=BCS_BoneSpace, alphaBoolBlend {blendOption=Linear}, boneToModify {boneName=dyn_L_feather_2}}
   n6 = "Transform (Modify) Bone - Bone: dyn_L_feather_3"(ComponentPose=@n5, Alpha=1, Translation=Get L FloaterOutC() -> (L FloaterOutC:Vector), Rotation="0,0,0", Scale="1,1,1") {translationMode=BMM_Additive, translationSpace=BCS_BoneSpace, alphaBoolBlend {blendOption=Linear}, boneToModify {boneName=dyn_L_feather_3}}
   Output Pose(Result=Component To Local(ComponentPose=@n6)) {layerGroup=None}
  ]}
 EventGraph {nodes [
   Event BlueprintUpdateAnimation()
   n1 = GetRealTimeSeconds()
   n2 = "Sin (Radians)"(A=float * float(A=@n1, B=1.5))
   n3 = "Sin (Radians)"(A=float * float(A=@n1, B=1.4))
   > "Set R FloaterOut A "("R FloaterOut A ":Vector=MakeVector(X=float * float(A=@n2, B=0.8), Y=float * float(A=@n3, B=0.7))) -> (Output_Get:Vector)
   n4 = float + float(A=@n1, B=0.9)
   n5 = "Sin (Radians)"(A=float * float(A=@n4, B=1.5))
   n6 = "Sin (Radians)"(A=float * float(A=@n4, B=1.4))
   > "Set L FloaterOut A "("L FloaterOut A ":Vector=MakeVector(X=float * float(A=@n5, B=0.8), Y=float * float(A=@n6, B=0.7))) -> (Output_Get:Vector)
   n7 = "Sin (Radians)"(A=float * float(A=@n1, B=1.4))
   n8 = "Sin (Radians)"(A=float * float(A=@n1, B=1.3))
   > Set R FloaterOut B(R FloaterOut B:Vector=MakeVector(X=float * float(A=@n7, B=0.8), Y=float * float(A=@n8, B=0.7))) -> (Output_Get:Vector)
   n9 = float + float(A=@n1, B=0.9)
   n10 = "Sin (Radians)"(A=float * float(A=@n9, B=1.4))
   n11 = "Sin (Radians)"(A=float * float(A=@n9, B=1.3))
   > Set L FloaterOut B(L FloaterOut B:Vector=MakeVector(X=float * float(A=@n10, B=0.8), Y=float * float(A=@n11, B=0.7))) -> (Output_Get:Vector)
   n12 = "Sin (Radians)"(A=float * float(A=@n1, B=1.3))
   n13 = "Sin (Radians)"(A=float * float(A=@n1, B=1.2))
   > Set R FloaterOutC(R FloaterOutC:Vector=MakeVector(X=float * float(A=@n12, B=0.8), Y=float * float(A=@n13, B=0.7))) -> (Output_Get:Vector)
   n14 = float + float(A=@n1, B=0.9)
   n15 = "Sin (Radians)"(A=float * float(A=@n14, B=1.3))
   n16 = "Sin (Radians)"(A=float * float(A=@n14, B=1.2))
   > Set L FloaterOutC(L FloaterOutC:Vector=MakeVector(X=float * float(A=@n15, B=0.8), Y=float * float(A=@n16, B=0.7))) -> (Output_Get:Vector)
  ]}
]
variables [
 "L FloaterOut A " {type=Vector}
 L FloaterOut B {type=Vector}
 L FloaterOutC {type=Vector}
 "R FloaterOut A " {type=Vector}
 R FloaterOut B {type=Vector}
 R FloaterOutC {type=Vector}
]
